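# Editor Utility Widget that serves as a row in the Events Validator tool, showing an asset name and a status label.
# A dark bordered layout places the two text blocks either side of a spacer, with the inner border, name and status exposed as variables.
# It declares AnimSequence and sequence binding variables plus a "Create New AnimSequence for Binding" graph.
# That graph and both construct events contain no logic here.
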EditorUtilityWidgetBlueprint /AnimInEdTools/EventsValidator/SubTools/EUW_EventsValidator_Row
graphs [
 Create New AnimSequence for Binding {nodes [Create New AnimSequence for Binding!() -> (then, Binding:Movie Scene Binding Proxy Structure) {is=title K2Node_FunctionEntry}]}
 EventGraph {nodes [
   Event PreConstruct()
   Event Construct()
  ]}
]
variables [
 AnimSequence {type=AnimSequence, default=null}
 Binding {type=string, default=""}
 Create New AnimSequence for Binding {default=()}
 In Binding {type=MovieSceneBindingProxy, default {bindingId=#1, sequence=null}}
]
widgets {
 info {rootWidgetClass=&Border, widgetCount=8}
 widgets [
  Border_178 {widgetClassPath=&Border, widget=Border_178, properties {brushColor=<rgba 0.0103 0.0103 0.0103 1>, padding=<ltrb 20 5 20 1>}}
  ColorBorder {
   widgetClassPath=&Border
   bIsVariable=true
   parent=Border_178
   widget=ColorBorder
   properties {brushColor=<rgba 0.0417 0.0417 0.0417 1>, padding=<ltrb 0 5 0 5>}
   slot {class=BorderSlot, properties {padding=<ltrb 20 5 20 1>}}
  }
  HorizontalBox {
   widgetClassPath=&HorizontalBox
   parent=ColorBorder
   widget=HorizontalBox
   slot {class=BorderSlot, properties {padding=<ltrb 0 5 0 5>}}
  }
  HorizontalBox_47 {widgetClassPath=&HorizontalBox, parent=HorizontalBox, widget=HorizontalBox_47, slot {class=HorizontalBoxSlot}}
  AssetName {
   widgetClassPath=&TextBlock
   bIsVariable=true
   parent=HorizontalBox_47
   widget=AssetName
   properties {
    text=Text Block
    colorAndOpacity {colorUseRule=UseColor_Specified, specifiedColor=<rgba 0.525 0.525 0.525 1>}
    font {
     bForceMonospaced=false
     bMaterialIsStencil=false
     fontMaterial=null
     fontObject=&Roboto
     letterSpacing=0
     monospacedWidth=1
     size=11.25
     skewAmount=0
     typefaceFontName=Bold
     outlineSettings {
      bApplyOutlineToDropShadows=false
      bMiteredCorners=false
      bSeparateFillAlpha=false
      outlineMaterial=null
      outlineSize=0
      outlineColor=<rgba 0 0 0 1>
     }
    }
   }
   slot {class=HorizontalBoxSlot, properties {horizontalAlignment=HAlign_Center, verticalAlignment=VAlign_Center}}
  }
  Spacer_59 {
   widgetClassPath=&Spacer
   parent=HorizontalBox
   widget=Spacer_59
   slot {class=HorizontalBoxSlot, properties {padding=<ltrb 10 0 10 0>}}
  }
  HorizontalBox_617 {widgetClassPath=&HorizontalBox, parent=HorizontalBox, widget=HorizontalBox_617, slot {class=HorizontalBoxSlot}}
  StatusText {
   widgetClassPath=&TextBlock
   bIsVariable=true
   parent=HorizontalBox_617
   widget=StatusText
   properties {
    text=Text Block
    colorAndOpacity {colorUseRule=UseColor_Specified, specifiedColor=<rgba 0.5104 0.5104 0.5104 1>}
    font {
     bForceMonospaced=false
     bMaterialIsStencil=false
     fontMaterial=null
     fontObject=&Roboto
     letterSpacing=0
     monospacedWidth=1
     size=10
     skewAmount=0
     typefaceFontName=Bold
     outlineSettings {
      bApplyOutlineToDropShadows=false
      bMiteredCorners=false
      bSeparateFillAlpha=false
      outlineMaterial=null
      outlineSize=0
      outlineColor=<rgba 0 0 0 1>
     }
    }
   }
   slot {class=HorizontalBoxSlot, properties {horizontalAlignment=HAlign_Center, verticalAlignment=VAlign_Center}}
  }
 ]
}
