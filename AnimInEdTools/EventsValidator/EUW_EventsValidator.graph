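# Editor utility widget that validates the user's checked-out animation and level sequences against a selected Animation Browser project's event data.
# An embedded Python script lists Perforce-opened assets, then the project's root sequence is opened, with a timer waiting for it.
# Recursion through subsequences and animation tracks feeds EventValidatorFunctions for sequence, section and work-in-progress checks.
# Results appear as color-coded rows with status text in a scroll box.
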EditorUtilityWidgetBlueprint /AnimInEdTools/EventsValidator/EUW_EventsValidator
graphs [
 CreateAnimSeqRow {nodes [
   n1 = CreateAnimSeqRow!() -> (then, AnimSequence:Anim Sequence Object Reference, StatusText:String, IsValid:bool=~) {is=title K2Node_FunctionEntry}
   > n2 = Branch(Condition=@n1.IsValid)
   > n3 = Set isValidColor(isValidColor:Linear Color Structure=(R=0.0067,G=0.3646,B=0,A=1)) -> (Output_Get:Linear Color Structure)
   n4 = Create Widget!(
    execute:Exec=[@n9.then, @n3.then]
    WorldContextObject:Object Reference
    Class:User Widget Class Reference=/AnimInEdTools/EventsValidator/SubTools/EUW_EventsValidator_Row.EUW_EventsValidator_Row_C
    OwningPlayer:Player Controller Object Reference
   ) -> (then:Exec, ReturnValue:EUW Events Validator Row Object Reference)
   > n5 = Set SectionRowRef(SectionRowRef:EUW Events Validator Row Object Reference=@n4) -> (Output_Get:EUW Events Validator Row Object Reference)
   n6 = Get AnimSequence() -> (AnimSequence:Anim Sequence Object Reference)
   > Set Text(
    Text="To Text (String)"(InString=GetDisplayName(Object=@n6))
    self=Get AssetName!(self:EUW Events Validator Row Object Reference=@n5) -> (AssetName:Text Object Reference)
   )
   n7 = Get SectionRowRef() -> (SectionRowRef:EUW Events Validator Row Object Reference)
   > Set members in Slate Brush!(
    execute:Exec
    StructRef:"Slate Brush Structure (by ref)"=Get Background!(self:Border Object Reference=Get ColorBorder!(self:EUW Events Validator Row Object Reference=@n7) -> (ColorBorder:Border Object Reference)) -> (Background:Slate Brush Structure)
    TintColor_SpecifiedColor:Linear Color Structure=(R=0,G=0.0198,B=0.276,A=1)
    TintColor_ColorUseRule:ESlateColorStylingMode Enum
   ) -> (then:Exec, StructOut:"Slate Brush Structure (by ref)") {is=title K2Node_SetFieldsInStruct}
   > Set Text(
    Text="To Text (String)"(InString=Get StatusText() -> (StatusText:String))
    self=Get StatusText!(self:EUW Events Validator Row Object Reference=Get SectionRowRef()) -> (StatusText:Text Object Reference)
   )
   n8 = Get SectionRowRef()
   > Set members in Slate Color!(
    execute:Exec
    StructRef:"Slate Color Structure (by ref)"=Get ColorAndOpacity!(self:Text Object Reference=Get AssetName(self=@n8)) -> (ColorAndOpacity:Slate Color Structure)
    SpecifiedColor:Linear Color Structure=Get isValidColor() -> (isValidColor:Linear Color Structure)
   ) -> (then:Exec, StructOut:"Slate Color Structure (by ref)") {is=title K2Node_SetFieldsInStruct}
   > AddChild(self=Get ScrollBox() -> (ScrollBox:Editor Utility Scroll Box Object Reference), Content=Get SectionRowRef())
   n9 = Set isValidColor(execute=@n2.else, isValidColor=(R=0.3646,G=0.0103,B=0,A=1))
  ]}
 CreateAnimSeqSectionRow {nodes [
   n1 = CreateAnimSeqSectionRow!() -> (then, SectionSequence:Anim Sequence Object Reference, StatusText:String, IsValid:bool=~) {is=title K2Node_FunctionEntry}
   > n2 = Branch(Condition=@n1.IsValid)
   > n3 = Set isValidColor(isValidColor=(R=0.0067,G=0.3646,B=0,A=1))
   n4 = Create Widget(execute=[@n9.then, @n3.then], Class=&EUW_EventsValidator_Row_C)
   > n5 = Set SectionRowRef(SectionRowRef=@n4)
   n6 = GetDisplayName(Object=Get SectionSequence() -> (SectionSequence:Anim Sequence Object Reference))
   > Set Text(Text="To Text (String)"(InString=Append(A=@n6, B=_Section)), self=Get AssetName(self=@n5))
   n7 = Get SectionRowRef()
   > Set members in Slate Brush(StructRef=Get Background(self=Get ColorBorder(self=@n7)), TintColor_SpecifiedColor=(R=0,G=0.4116,B=0.4688,A=1))
   > Set Text(
    Text="To Text (String)"(InString=Get StatusText!() -> (StatusText:String))
    self=Get StatusText!(self:EUW Events Validator Row Object Reference=Get SectionRowRef()) -> (StatusText:Text Object Reference)
   )
   n8 = Get SectionRowRef()
   > Set members in Slate Color(StructRef=Get ColorAndOpacity(self=Get AssetName(self=@n8)), SpecifiedColor=Get isValidColor())
   > AddChild(self=Get ScrollBox(), Content=Get SectionRowRef())
   n9 = Set isValidColor(execute=@n2.else, isValidColor=(R=0.3646,G=0.0103,B=0,A=1))
  ]}
 CreateLvlSeqRow {nodes [
   n1 = CreateLvlSeqRow!() -> (then, LevelSequence:Level Sequence Object Reference, StatusText:String, IsValid:bool=~) {is=title K2Node_FunctionEntry}
   > n2 = Branch(Condition=@n1.IsValid)
   > n3 = Set isValidColor(isValidColor=(R=0.0067,G=0.3646,B=0,A=1))
   n4 = Create Widget(execute=[@n9.then, @n3.then], Class=&EUW_EventsValidator_Row_C)
   > n5 = Set SectionRowRef(SectionRowRef=@n4)
   n6 = Get LevelSequence() -> (LevelSequence:Level Sequence Object Reference)
   > Set Text(Text="To Text (String)"(InString=GetDisplayName(Object=@n6)), self=Get AssetName(self=@n5))
   n7 = Get SectionRowRef()
   > Set members in Slate Brush(StructRef=Get Background(self=Get ColorBorder(self=@n7)), TintColor_SpecifiedColor=(R=0.2459,G=0,B=0.276,A=1))
   > Set Text(
    Text="To Text (String)"(InString=Get StatusText!() -> (StatusText:String))
    self=Get StatusText!(self:EUW Events Validator Row Object Reference=Get SectionRowRef()) -> (StatusText:Text Object Reference)
   )
   n8 = Get SectionRowRef()
   > Set members in Slate Color(StructRef=Get ColorAndOpacity(self=Get AssetName(self=@n8)), SpecifiedColor=Get isValidColor())
   > AddChild(self=Get ScrollBox(), Content=Get SectionRowRef())
   n9 = Set isValidColor(execute=@n2.else, isValidColor=(R=0.3646,G=0.0103,B=0,A=1))
  ]}
 CreateLvlSeqSectionRow {nodes [
   n1 = CreateLvlSeqSectionRow!() -> (then, SectionSequence:Level Sequence Object Reference, StatusText:String, IsValid:bool=~) {is=title K2Node_FunctionEntry}
   > n2 = Branch(Condition=@n1.IsValid)
   > n3 = Set isValidColor(isValidColor=(R=0.0067,G=0.3646,B=0,A=1))
   n4 = Create Widget(execute=[@n9.then, @n3.then], Class=&EUW_EventsValidator_Row_C)
   > n5 = Set SectionRowRef(SectionRowRef=@n4)
   n6 = GetDisplayName(Object=Get SectionSequence() -> (SectionSequence:Level Sequence Object Reference))
   > Set Text(Text="To Text (String)"(InString=Append(A=@n6, B=_Section)), self=Get AssetName(self=@n5))
   n7 = Get SectionRowRef()
   > Set members in Slate Brush(StructRef=Get Background(self=Get ColorBorder(self=@n7)), TintColor_SpecifiedColor=(R=0.276,G=0,B=0.2448,A=1))
   > Set Text(
    Text="To Text (String)"(InString=Get StatusText!() -> (StatusText:String))
    self=Get StatusText!(self:EUW Events Validator Row Object Reference=Get SectionRowRef()) -> (StatusText:Text Object Reference)
   )
   n8 = Get SectionRowRef()
   > Set members in Slate Color(StructRef=Get ColorAndOpacity(self=Get AssetName(self=@n8)), SpecifiedColor=Get isValidColor())
   > AddChild(self=Get ScrollBox(), Content=Get SectionRowRef())
   n9 = Set isValidColor(execute=@n2.else, isValidColor=(R=0.3646,G=0.0103,B=0,A=1))
  ]}
 EventGraph {nodes [
   n1 = "On Opening (Project_ComboBox)"() {is=title K2Node_ComponentBoundEvent}
   UpdateProjectComboBox!(
    execute=[@n1.then, @n6.then]
    self
    ComboBox:"Searchable Combo Box (String) Object Reference"=Get Project_ComboBox() -> (Project_ComboBox:"Searchable Combo Box (String) Object Reference")
    Items:Array of Strings=Get ProjectNames() -> (ProjectNames:Array of Strings)
    DefaultItem:String
    AddBlank:bool=false
   ) -> (then)
   "On Clicked (ValidateBtn)"() {is=title K2Node_ComponentBoundEvent}
   > ClearChildren(self=Get ScrollBox())
   > Clear(TargetArray=Get CheckedSequences() -> (CheckedSequences:Array of Strings))
   > GetModifiedAssets()
   n2 = "On Selection Changed (Project_ComboBox)"!() -> (OutputDelegate, then, SelectedItem:String, SelectionType:ESelectInfo Enum) {is=title K2Node_ComponentBoundEvent}
   > ClearChildren(self=Get ScrollBox())
   n3 = Get ProjectNames()
   > Branch(Condition="integer > integer"(A=Length(TargetArray:Array of Strings=@n3)))
   > n4 = Set Current Project(Current Project:BP Animation Browser Project Object Reference="Get (a copy)"(
     Array:Array of BP Animation Browser Project Object References=Get CurrentProjects() -> (CurrentProjects:Array of BP Animation Browser Project Object References)
     Dimension 1=Find Item!(TargetArray=@n3, ItemToFind:"String (by ref)"=@n2.SelectedItem) -> (ReturnValue)
    ) -> (Output:BP Animation Browser Project Object Reference)) -> (Output_Get:BP Animation Browser Project Object Reference)
   > Is Valid(InputObject=@n4)
   > UpdateComboBox'()
   > n5 = For Each Loop(Array:Array of Animatable Creature Object References=Get Animatable Creatures!(
     Project:BP Animation Browser Project Object Reference=Get Current Project() -> (Current Project:BP Animation Browser Project Object Reference)
     bSearchSubClasses:bool=false
     __WorldContext:Object Reference
    ) -> (Animatable Creatures:Array of Animatable Creature Object References) {is=title K2Node_CallFunction}) -> (Array Element:Animatable Creature Object Reference)
   > Set ProjectAC(ProjectAC:Animatable Creature Object Reference=@n5) -> (Output_Get:Animatable Creature Object Reference)
   n6 = UpdateComboBox'() {is=title K2Node_CustomEvent}
   Event Construct()
   > n7 = Set Timer by Function Name!(
    execute
    Object:Object Reference
    FunctionName:String=PostConstruct
    Time:float=0.2
    bLooping:bool=true
    bMaxOncePerFrame:bool=false
    InitialStartDelay:float=0
    InitialStartDelayVariance:float=0
   ) -> (then, ReturnValue:Timer Handle Structure) {is=title K2Node_CallFunction}
   > Set ConstructTimer(ConstructTimer:Timer Handle Structure=@n7) -> (Output_Get:Timer Handle Structure)
   Event PreConstruct()
   Event Tick()
  ]}
 GetModifiedAssets {nodes [
   GetModifiedAssets() {is=title K2Node_FunctionEntry}
   > n1 = Execute Python Script!(
    execute
    PythonScript="import subprocess\r\nimport configparser\r\nimport os\r\nimport re\r\n\r\n\r\ndef get_workspace_from_editor():\r\n    \"\"\"\r\n    Read the P4 workspace name from the editor's own SourceControlSettings.ini.\r\n    This is the same workspace the editor is connected to — works for every user\r\n    without any configuration.\r\n    \"\"\"\r\n    try:\r\n        import unreal\r\n        ini_path = os.path.join(\r\n            unreal.Paths.project_saved_dir(),\r\n            \"Config\", \"WindowsEditor\", \"SourceControlSettings.ini\"\r\n        )\r\n    except ImportError:\r\n        ## Running outside the editor — try common locations\r\n        ini_path = None\r\n\r\n    if ini_path and os.path.exists(ini_path):\r\n        config = configparser.ConfigParser()\r\n        config.read(ini_path)\r\n        section = \"PerforceSourceControl.PerforceSourceControlSettings\"\r\n        workspace = config.get(section, \"Workspace\", fallback=\"\")\r\n        if workspace:\r\n            return workspace\r\n\r\n    return None\r\n\r\n\r\ndef get_workspace_from_p4():\r\n    \"\"\"Fallback: read P4CLIENT from the p4 environment.\"\"\"\r\n    try:\r\n        result = subprocess.run(\r\n            [\"p4\", \"set\", \"P4CLIENT\"],\r\n            capture_output=True, text=True, timeout=10\r\n        )\r\n        for line in result.stdout.strip().splitlines():\r\n            match = re.match(r\"P4CLIENT=(\\S+)\", line)\r\n            if match:\r\n                val = match.group(1)\r\n                if val and val != \"(set)\":\r\n                    return val\r\n    except Exception:\r\n        pass\r\n    return None\r\n\r\n\r\ndef resolve_workspace():\r\n    \"\"\"\r\n    Resolve the workspace name. Priority:\r\n    1. Editor INI (most reliable — matches what the editor is connected to)\r\n    2. P4 environment variable (fallback)\r\n    \"\"\"\r\n    ws = get_workspace_from_editor()\r\n    if ws:\r\n        print(f\"[P4] Workspace from editor settings: {ws}\")\r\n        return ws\r\n\r\n    ws = get_workspace_from_p4()\r\n    if ws:\r\n        print(f\"[P4] Workspace from P4 environment: {ws}\")\r\n        return ws\r\n\r\n    print(\"[P4 Warning] Could not detect workspace. Using P4 defaults.\")\r\n    return None\r\n\r\n\r\ndef run_p4_command(args, workspace=None):\r\n    \"\"\"Run a p4 command and return stdout lines. Returns empty list on failure.\"\"\"\r\n    try:\r\n        cmd = [\"p4\"]\r\n        if workspace:\r\n            cmd += [\"-c\", workspace]\r\n        cmd += args\r\n\r\n        result = subprocess.run(cmd, capture_output=True, text=True, timeout=30)\r\n        if result.returncode != 0:\r\n            error_msg = result.stderr.strip()\r\n            if \"not opened\" in error_msg.lower():\r\n                return []\r\n            print(f\"[P4 Error] {' '.join(cmd)}: {error_msg}\")\r\n            return []\r\n        return [line for line in result.stdout.strip().splitlines() if line]\r\n    except FileNotFoundError:\r\n        print(\"[P4 Error] 'p4' command not found. Is Perforce installed and on PATH?\")\r\n        return []\r\n    except subprocess.TimeoutExpired:\r\n        print(\"[P4 Error] p4 command timed out after 30s.\")\r\n        return []\r\n    except Exception as e:\r\n        print(f\"[P4 Error] Unexpected error: {e}\")\r\n        return []\r\n\r\n\r\ndef parse_opened_line(line):\r\n    \"\"\"\r\n    Parse a line from 'p4 opened' output.\r\n    Format: //depot/path/to/file.uasset##3 - edit default change (binary+l)\r\n    Returns the depot path, or None if unparseable.\r\n    \"\"\"\r\n    match = re.match(r\"^(//[^##]+)##\", line)\r\n    return match.group(1) if match else None\r\n\r\n\r\ndef depot_to_local(depot_paths, workspace=None):\r\n    \"\"\"Convert depot paths to local filesystem paths using 'p4 where'.\"\"\"\r\n    if not depot_paths:\r\n        return {}\r\n\r\n    mapping = {}\r\n    lines = run_p4_command([\"where\"] + depot_paths, workspace)\r\n\r\n    for line in lines:\r\n        ## p4 where output: //depot/path //workspace/path C:\\local\\path\r\n        parts = re.split(r\"\\s+//\", line)\r\n        if len(parts) >= 3:\r\n            depot = \"//\" + parts[0].lstrip(\"/\")\r\n            local = parts[-1].strip()\r\n            mapping[depot] = local\r\n        else:\r\n            parts = line.split()\r\n            if len(parts) >= 3:\r\n                mapping[parts[0]] = parts[-1]\r\n\r\n    return mapping\r\n\r\n\r\ndef get_checked_out_assets(asset_extensions=None):\r\n    \"\"\"\r\n    Returns a list of full local filesystem paths for all files checked out in the\r\n    current editor's Perforce workspace.\r\n\r\n    Args:\r\n        asset_extensions: Set of extensions to include. Defaults to {\".uasset\", \".umap\"}.\r\n                         Pass empty set to include ALL file types.\r\n\r\n    Returns:\r\n        List of strings — full local filesystem paths (e.g. C:/Workspace/Content/...).\r\n        Falls back to depot path if local path cannot be resolved.\r\n    \"\"\"\r\n    if asset_extensions is None:\r\n        asset_extensions = {\".uasset\", \".umap\"}\r\n\r\n    ## Step 1: Resolve workspace (auto-detect from editor)\r\n    ws = resolve_workspace()\r\n\r\n    ## Step 2: Get all checked-out files\r\n    opened_lines = run_p4_command([\"opened\"], ws)\r\n    if not opened_lines:\r\n        print(\"[P4] No checked-out files found.\")\r\n        return []\r\n\r\n    print(f\"[P4] Found {len(opened_lines)} checked-out file(s).\")\r\n\r\n    ## Step 3: Parse depot paths\r\n    depot_paths = []\r\n    for line in opened_lines:\r\n        depot = parse_opened_line(line)\r\n        if depot:\r\n            depot_paths.append(depot)\r\n\r\n    if not depot_paths:\r\n        return []\r\n\r\n    ## Step 4: Filter to asset files only\r\n    if asset_extensions:\r\n        depot_paths = [\r\n            p for p in depot_paths\r\n            if any(p.lower().endswith(ext) for ext in asset_extensions)\r\n        ]\r\n        if not depot_paths:\r\n            print(\"[P4] No asset files among checked-out files.\")\r\n            return []\r\n\r\n    print(f\"[P4] {len(depot_paths)} asset file(s) after filtering.\")\r\n\r\n    ## Step 5: Convert depot -> local paths\r\n    depot_to_local_map = depot_to_local(depot_paths, ws)\r\n\r\n    asset_paths = []\r\n    for depot in depot_paths:\r\n        local = depot_to_local_map.get(depot)\r\n        if local:\r\n            asset_paths.append(local)\r\n        else:\r\n            asset_paths.append(depot)\r\n\r\n    return asset_paths\r\n\r\n\r\n## ----- Blueprint Entry Point -----\r\n## Output pin: checked_out_assets (Array of Strings)\r\n## No input needed — workspace is auto-detected from the editor's P4 settings.\r\nchecked_out_assets = get_checked_out_assets()"
   ) -> (then, ReturnValue, checked_out_assets:Array of Strings) {outputs [checked_out_assets]}
   > n2 = For Each Loop(Array:Array of Strings=@n1.checked_out_assets) -> (Array Element:String)
   n3 = Create Widget(execute=@n2.Completed, Class=/AnimInEdTools/EventsValidator/SubTools/EventValidatorFunctions.EventValidatorFunctions_C) -> (ReturnValue:Event Validator Functions Object Reference)
   > Is Valid(InputObject=@n3)
   > Set Validator(Validator:Event Validator Functions Object Reference=@n3) -> (Output_Get:Event Validator Functions Object Reference)
   > GetProjectRootSequence()
   n4 = FilenameToPackageName!(execute=@n2.LoopBody, Filename:String=ChangeExtension!(InPath:String=@n2, InNewExtension:String) -> (ReturnValue:String)) -> (then, ReturnValue:String)
   > n5 = LoadAsset(AssetPath=@n4)
   > Is Valid(InputObject=@n5)
   > Add!(
    execute
    TargetMap:Map of Strings to Object References=Get CheckedOutAssets() -> (CheckedOutAssets:Map of Strings to Object References)
    Key:"String (by ref)"=GetDisplayName(Object=@n5)
    Value:"Object Reference (by ref)"=@n5
   ) -> (then) {is=title K2Node_CallFunction}
   PrintString(InString=Failed to make widget)
   PrintString(InString=Failed to load checked out asset)
  ]}
 GetProjectRootSequence {nodes [
   GetProjectRootSequence() {is=title K2Node_FunctionEntry}
   n1 = Get ProjectAC() -> (ProjectAC:Animatable Creature Object Reference)
   > n2 = GetAssetsByPath!(
    execute
    self:Asset Registry Interface=GetAssetRegistry()
    PackagePath:Name=String To Name(InString=Get AnimMontage Path!(self:Animatable Creature Object Reference=@n1) -> (AnimMontage Path:String))
    bRecursive:bool=false
    bIncludeOnlyOnDiskAssets:bool=false
   ) -> (then, OutAssetData:Array of Asset Data Structures, ReturnValue:bool)
   > n3 = For Each Loop!(Exec, Array:Array of Asset Data Structures=@n2) -> (
    LoopBody
    Array Element_PackageName:Name
    Array Element_PackagePath:Name
    Array Element_AssetName:Name
    Array Element_AssetClassPath:Top Level Asset Path Structure
    Array Index
    Completed
   )
   n4 = "To String (Name)"(InName=@n3.Array Element_AssetName)
   n5 = Get Current Project()
   > Branch(Condition=Contains(SearchIn=@n4, Substring=Get Name!(self:BP Animation Browser Project Object Reference=@n5) -> (Name:String)))
   > n6 = LoadAsset(AssetPath=Append!(A="To String (Name)"(InName=@n3), B=., C:String=@n4) -> (ReturnValue))
   > n7 = Cast To LevelSequence!(execute, Object=@n6) -> (then, CastFailed, AsLevel Sequence:Level Sequence Object Reference) {is=title K2Node_DynamicCast}
   > n8 = Set ProjectRootSequence(ProjectRootSequence:Level Sequence Object Reference=@n7) -> (Output_Get:Level Sequence Object Reference)
   > OpenLevelSequence!(execute, LevelSequence:Level Sequence Object Reference=@n8) -> (then, ReturnValue:bool)
   > n9 = Set Timer by Function Name(FunctionName=WaitForOpen, Time=0.1, bLooping=true, bMaxOncePerFrame=false, InitialStartDelay=0, InitialStartDelayVariance=0)
   > Set Timer(Timer:Timer Handle Structure=@n9) -> (Output_Get:Timer Handle Structure)
   PrintString(InString=FAILED TO GET PROJECT LEVEL SEQUENCE)
   PrintString(InString=FAILED TO FIND EVENT)
  ]}
 GetProjects {nodes [
   GetProjects() {is=title K2Node_FunctionEntry}
   > Clear(TargetArray:Array of BP Animation Browser Project Object References=Get Local Projects() -> (Local Projects:Array of BP Animation Browser Project Object References))
   > n1 = GetAssetsByClass!(
    execute
    self:Asset Registry Interface=GetAssetRegistry()
    ClassPathName:Top Level Asset Path Structure=GetSoftClassTopLevelAssetPath!(SoftClassReference:Object Soft Class Reference=/AnimInEdTools/AnimationBrowser/ProjectSettings/BP_AnimationBrowser_Project.BP_AnimationBrowser_Project_C) -> (ReturnValue:Top Level Asset Path Structure)
    bSearchSubClasses:bool=false
   ) -> (then, OutAssetData:Array of Asset Data Structures, ReturnValue:bool)
   > n2 = For Each Loop!(Exec, Array=@n1) -> (LoopBody, Array Element:Asset Data Structure, Array Index, Completed)
   Return Node!(execute=@n2.Completed, Projects:Array of BP Animation Browser Project Object References=Get Local Projects())
   n3 = Break Asset Data(AssetData=@n2)
   > n4 = LoadAsset(AssetPath=GetBaseFilename(InPath="To String (Name)"(InName=@n3), bRemovePath=false))
   > n5 = Cast To BP_AnimationBrowser_Project!(execute, Object=@n4) -> (then, CastFailed, AsBP Animation Browser Project:BP Animation Browser Project Object Reference) {is=title K2Node_DynamicCast}
   > Add(
    TargetArray:Array of BP Animation Browser Project Object References=Get Local Projects()
    NewItem:"BP Animation Browser Project Object Reference (by ref)"=@n5
   ) {is=title K2Node_CallArrayFunction}
  ]}
 GetSequenceAnimTracks {nodes [
   n1 = GetSequenceAnimTracks!() -> (then, Sequence:Movie Scene Sequence Object Reference) {is=title K2Node_FunctionEntry}
   > n2 = GetBindings!(execute, Sequence:Movie Scene Sequence Object Reference=@n1) -> (then, ReturnValue:Array of Movie Scene Binding Proxy Structures)
   > n3 = For Each Loop(Array:Array of Movie Scene Binding Proxy Structures=@n2) -> (Array Element:Movie Scene Binding Proxy Structure)
   > n4 = GetTracks!(execute, InBinding:"Movie Scene Binding Proxy Structure (by ref)"=@n3) -> (then, ReturnValue:Array of Movie Scene Track Object References)
   > n5 = For Each Loop(Array:Array of Movie Scene Track Object References=@n4) -> (Array Element:Movie Scene Track Object Reference)
   > n6 = Cast To MovieSceneSkeletalAnimationTrack!(execute, Object=@n5) -> (then, CastFailed, AsMovie Scene Skeletal Animation Track:Movie Scene Skeletal Animation Track Object Reference) {is=title K2Node_DynamicCast}
   > n7 = GetSections!(execute, Track:Movie Scene Track Object Reference=@n6) -> (then, ReturnValue:Array of Movie Scene Section Object References)
   > n8 = For Each Loop(Array:Array of Movie Scene Section Object References=@n7) -> (Array Element:Movie Scene Section Object Reference)
   > n9 = Cast To MovieSceneSkeletalAnimationSection!(execute, Object=@n8) -> (then, CastFailed, AsMovie Scene Skeletal Animation Section:Movie Scene Skeletal Animation Section Object Reference) {is=title K2Node_DynamicCast}
   > n10 = Cast To AnimSequence!(
    execute
    Object=Get!(self:Movie Scene Skeletal Animation Section Object Reference=@n9) -> (
     Params_Animation:Anim Sequence Base Object Reference
     Params_FirstLoopStartFrameOffset:Frame Number Structure
     Params_StartFrameOffset:Frame Number Structure
     Params_EndFrameOffset:Frame Number Structure
     Params_PlayRate:Movie Scene Time Warp Variant Structure
     Params_bReverse:bool=~
     Params_SlotName:Name
     Params_MirrorDataTable:Mirror Data Table Object Reference
     Params_bSkipAnimNotifiers:bool=~
     Params_bForceCustomMode:bool=~
     Params_SwapRootBone:ESwapRootBone Enum
     Params_bLinearPlaybackWhenScaled:bool=~
    )
   ) -> (then, CastFailed, AsAnim Sequence:Anim Sequence Object Reference)
   > Branch(Condition=Find(TargetMap:Map of Strings to Object References=Get CheckedOutAssets(), Key=GetDisplayName(Object=@n10)) -> (Value:Object Reference).ReturnValue)
   n11 = ValidateAnimSeqSection!(
    self:Event Validator Functions Object Reference=Get Validator() -> (Validator:Event Validator Functions Object Reference)
    Project:BP Animation Browser Project Object Reference=Get Current Project()
    Section:Movie Scene Skeletal Animation Section Object Reference=@n9
    AnimSequence:Object Reference=@n10
   ) -> (IsValid:bool, StatusText:String)
   > CreateAnimSeqSectionRow!(execute, self, SectionSequence:Anim Sequence Object Reference=@n10, StatusText:String=@n11.StatusText, IsValid:bool=@n11) -> (then) {is=title K2Node_CallFunction}
   n12 = ValidateAnimSeq!(
    self:Event Validator Functions Object Reference=Get Validator()
    Project:BP Animation Browser Project Object Reference=Get Current Project()
    AnimSequence:Anim Sequence Object Reference=@n10
   ) -> (IsValid:bool, RowFound:bool, RequiredStartFrame:int=0, RequiredEndFrame:int=0, StatusText:String)
   > CreateAnimSeqRow!(execute, self, AnimSequence:Anim Sequence Object Reference=@n10, StatusText:String=@n12.StatusText, IsValid:bool=@n12) -> (then) {is=title K2Node_CallFunction}
  ]}
 GetSubSequences {nodes [
   n1 = GetSubSequences!() -> (then, Sequence:Movie Scene Sequence Object Reference) {is=title K2Node_FunctionEntry}
   > n2 = GetTracks!(execute, Sequence:Movie Scene Sequence Object Reference=@n1) -> (then, ReturnValue)
   > n3 = For Each Loop(Array:Array of Movie Scene Track Object References=@n2) -> (Array Element:Movie Scene Track Object Reference)
   GetSequenceAnimTracks!(execute=@n3.Completed, self, Sequence:Movie Scene Sequence Object Reference=@n1) -> (then) {is=title K2Node_CallFunction}
   > n4 = For Each Loop(Array:Array of Movie Scene Sequence Object References=Get SubSequences() -> (SubSequences:Array of Movie Scene Sequence Object References)) -> (Array Element:Movie Scene Sequence Object Reference)
   > GetSubSequences!(execute, self, Sequence:Movie Scene Sequence Object Reference=@n4) -> (then) {is=title K2Node_CallFunction}
   n5 = Cast To MovieSceneSubTrack!(execute=@n3.LoopBody, Object=@n3) -> (then, CastFailed, AsMovie Scene Sub Track:Movie Scene Sub Track Object Reference) {is=title K2Node_DynamicCast}
   > n6 = GetSections(Track=@n5)
   > n7 = For Each Loop(Array:Array of Movie Scene Section Object References=@n6) -> (Array Element:Movie Scene Section Object Reference)
   > n8 = Cast To MovieSceneSubSection!(execute, Object=@n7) -> (then, CastFailed, AsMovie Scene Sub Section:Movie Scene Sub Section Object Reference) {is=title K2Node_DynamicCast}
   > Set SubtrackSectionRef(SubtrackSectionRef:Movie Scene Sub Section Object Reference=@n8) -> (Output_Get:Movie Scene Sub Section Object Reference)
   n9 = GetSequence!(self:Movie Scene Sub Section Object Reference=@n8) -> (ReturnValue:Movie Scene Sequence Object Reference)
   > Add Unique(TargetArray:Array of Movie Scene Sequence Object References=Get SubSequences(), NewItem:"Movie Scene Sequence Object Reference (by ref)"=@n9)
   > n10 = Cast To LevelSequence(Object=@n9)
   > n11 = GetPlaybackRange!(execute, Sequence:Movie Scene Sequence Object Reference=@n10) -> (then, ReturnValue_bHasStart:bool=~, ReturnValue_bHasEnd:bool=~, ReturnValue_InclusiveStart:int, ReturnValue_ExclusiveEnd:int)
   > Set SubSeqRef(SubSeqRef:Level Sequence Object Reference=@n10) -> (Output_Get:Level Sequence Object Reference)
   > Set StartFrame(StartFrame:int=@n11.ReturnValue_InclusiveStart) -> (Output_Get:int=0)
   > Set EndFrame(EndFrame:int=@n11.ReturnValue_ExclusiveEnd) -> (Output_Get:int=0)
   > n12 = Sequence()
   n13 = ValidateLvlSeq!(
    self:Event Validator Functions Object Reference=Get Validator()
    Project:BP Animation Browser Project Object Reference=Get Current Project()
    LevelSequence:Level Sequence Object Reference=@n10
   ) -> (IsValid:bool, StatusText:String)
   > n14 = Branch(Condition=@n13)
   CreateLvlSeqRow!(execute=@n14.else, self, LevelSequence:Level Sequence Object Reference=@n10, StatusText:String=@n13.StatusText, IsValid:bool=@n13) -> (then) {is=title K2Node_CallFunction}
   n15 = Keys!(execute=@n12.then_1, TargetMap:Map of Strings to Object References=Get CheckedOutAssets()) -> (then, Keys:Array of Strings)
   > n16 = For Each Loop(Array:Array of Strings=@n15) -> (Array Element:String)
   n17 = Contains Item(TargetArray:Array of Strings=Get CheckedSequences(), ItemToFind:"String (by ref)"=@n16)
   n18 = GetDisplayName(Object=Get SubSeqRef() -> (SubSeqRef:Level Sequence Object Reference))
   > Branch(Condition=AND Boolean(A=Contains(SearchIn=@n16, Substring=@n18), B=NOT Boolean(A=@n17)))
   > Add(TargetArray:Array of Strings=Get CheckedSequences(), NewItem:"String (by ref)"=@n16)
   > n19 = Cast To LevelSequence(Object=Find(TargetMap=Get CheckedOutAssets(), Key=@n16))
   n20 = ValidateWIPSeq!(
    self:Event Validator Functions Object Reference=Get Validator()
    Project:BP Animation Browser Project Object Reference=Get Current Project()
    LevelSequence:Level Sequence Object Reference=@n19
    Range Start:int=Get StartFrame() -> (StartFrame:int=0)
    Range End:int=Get EndFrame() -> (EndFrame:int=0)
   ) -> (PartOfProject:bool, IsValid:bool, StatusText:String)
   > Branch(Condition=@n20)
   > CreateLvlSeqRow(LevelSequence=@n19, StatusText=@n20.StatusText, IsValid=@n20.IsValid)
  ]}
 PostConstruct {nodes [
   PostConstruct() {is=title K2Node_FunctionEntry}
   > n1 = Branch(Condition=IsLoadingAssets!(self:Asset Registry Interface=GetAssetRegistry()) -> (ReturnValue:bool))
   Clear and Invalidate Timer by Handle!(execute=@n1.else, WorldContextObject:Object Reference, Handle=Get ConstructTimer() -> (ConstructTimer:Timer Handle Structure)) -> (then)
   > n2 = Set CurrentProjects(CurrentProjects:Array of BP Animation Browser Project Object References=GetProjects!() -> (Projects:Array of BP Animation Browser Project Object References) {is=title K2Node_CallFunction}) -> (Output_Get:Array of BP Animation Browser Project Object References)
   > n3 = For Each Loop(Array:Array of BP Animation Browser Project Object References=@n2) -> (Array Element:BP Animation Browser Project Object Reference)
   > Add(TargetArray=Get ProjectNames(), NewItem=Get Name(self=@n3))
   > SetSelectedIndex!(execute, self:"Searchable Combo Box (String) Object Reference"=Get Project_ComboBox(), Index:int=0) -> (then)
   > UpdateComboBox'() {is=title K2Node_CallFunction}
  ]}
 UpdateProjectComboBox {nodes [
   n1 = UpdateProjectComboBox!() -> (then, ComboBox:"Searchable Combo Box (String) Object Reference", Items:Array of Strings, DefaultItem:String, AddBlank:bool=~) {is=title K2Node_FunctionEntry}
   > n2 = Set Current Index(Current Index:int=GetSelectedIndex!(self:"Searchable Combo Box (String) Object Reference"=@n1) -> (ReturnValue:int=0)) -> (Output_Get:int=0)
   > Set Selected Item(Selected Item:String=GetOptionAtIndex!(
     self:"Searchable Combo Box (String) Object Reference"=Get ComboBox() -> (ComboBox:"Searchable Combo Box (String) Object Reference")
     Index:int=@n2
    ) -> (ReturnValue:String)) -> (Output_Get:String)
   n3 = Get ComboBox()
   > n4 = For Loop(LastIndex=int - int(A=GetOptionCount!(self:"Searchable Combo Box (String) Object Reference"=@n3) -> (ReturnValue:int=0)))
   n5 = Branch(
    execute=@n4.Completed
    Condition=Identical!(ArrayA:Array of Strings=Get Items() -> (Items:Array of Strings), ArrayB:Array of Strings=Get Old Items() -> (Old Items:Array of Strings)) -> (ReturnValue:bool) {is=title K2Node_CallArrayFunction}
   )
   ClearOptions(execute=@n5.else, self:"Searchable Combo Box (String) Object Reference"=Get ComboBox())
   > n6 = Sequence!(execute) -> (then_0, then_1, then_2:Exec, then_3:Exec, then_4:Exec)
   > Branch(Condition=Get AddBlank() -> (AddBlank:bool))
   > AddOption!(execute, self:"Searchable Combo Box (String) Object Reference"=Get ComboBox(), Option:String) -> (then)
   n7 = Branch(execute=@n6.then_1, Condition=IsEmpty(InString=Get DefaultItem() -> (DefaultItem:String)))
   AddOption(execute=@n7.else, self=Get ComboBox())
   n8 = For Each Loop(Exec=@n6.then_2, Array:Array of Strings=Get Items()) -> (Array Element:String)
   > AddOption(self=Get ComboBox(), Option=@n8)
   n9 = Get Current Index() -> (Current Index:int=0)
   n10 = Get ComboBox()
   n11 = Branch(
    execute=@n6.then_3
    Condition="Equal Exactly (String)"(A=GetOptionAtIndex(self=@n10, Index=@n9), B=Get Selected Item() -> (Selected Item:String))
   )
   n12 = Get ComboBox()
   > AddOption(self=@n12, Option=Temp)
   n13 = Get Current Index() -> (Current Index=0)
   > SetSelectedIndex(self=@n12, Index=int + int(A=@n13))
   > SetSelectedIndex(self=@n12, Index=@n13)
   > RemoveOption!(execute, self:"Searchable Combo Box (String) Object Reference"=@n12, Option:String=Temp) -> (then, ReturnValue:bool)
   ClearSelection(execute=@n11.else, self:"Searchable Combo Box (String) Object Reference"=Get ComboBox())
   SetSelectedOption!(execute=@n6.then_4, self:"Searchable Combo Box (String) Object Reference"=Get ComboBox(), Option:String) -> (then)
   n14 = GetOptionAtIndex(self=Get ComboBox(), Index=@n4)
   n15 = Branch(execute=@n4.LoopBody, Condition="Equal Exactly (String)"(A=@n14, B=Get DefaultItem()))
   Add(execute=@n15.else, TargetArray=Get Old Items(), NewItem=@n14)
  ]}
 ValidateAnimSeqSection {nodes [
   n1 = ValidateAnimSeqSection!() -> (then, Section:Movie Scene Skeletal Animation Section Object Reference) {is=title K2Node_FunctionEntry}
   > n2 = Cast To AnimSequence(Object=Get(self=@n1) -> (Params_bReverse=~, Params_bSkipAnimNotifiers=~, Params_bForceCustomMode=~, Params_bLinearPlaybackWhenScaled=~))
   > n3 = Is Valid(InputObject=@n2)
   PrintString(execute=@n3.Is Not Valid, InString=anim sequence not valid)
   n4 = For Each Loop(Exec=@n3.Is Valid, Array:Array of Object References) -> (Array Element:Object Reference)
   > Branch(Condition="Equal Exactly (String)"(A=GetDisplayName(Object=@n4), B=GetDisplayName(Object=@n2)))
   > n5 = Cast To AnimSequence(Object=@n4)
   > Set AnimSeqRef(AnimSeqRef:Anim Sequence Object Reference=@n5) -> (Output_Get:Anim Sequence Object Reference)
   n6 = Get Section() -> (Section:Movie Scene Skeletal Animation Section Object Reference)
   > n7 = GetStartFrame!(execute, Section:Movie Scene Section Object Reference=@n6) -> (then, ReturnValue:int=0)
   > n8 = GetEndFrame!(execute, Section:Movie Scene Section Object Reference=@n6) -> (then, ReturnValue:int=0)
   > n9 = Sequence!(execute) -> (then_0, then_1)
   n10 = ValidateAnimSeq(self=Get Validator(), Project=Get Current Project(), AnimSequence=Get AnimSeqRef() -> (AnimSeqRef:Anim Sequence Object Reference)) -> (RequiredStartFrame=0, RequiredEndFrame=0)
   n11 = "Equal (Integer)"(A=@n7, B=@n10.RequiredStartFrame)
   n12 = "Equal (Integer)"(A=@n8, B=@n10.RequiredEndFrame)
   n13 = AND Boolean(A=@n11, B=@n12)
   > n14 = Branch(Condition=@n13)
   > n15 = Set AnimSectionStatus(AnimSectionStatus:String=Section is valid) -> (Output_Get:String)
   CreateAnimSeqSectionRow(
    execute=[@n17.then, @n15.then, @n18.then]
    SectionSequence=Get AnimSeqRef()
    StatusText=Get AnimSectionStatus() -> (AnimSectionStatus:String)
    IsValid=@n13
   )
   Set AnimSectionStatus(execute=@n14.else, AnimSectionStatus=Section is not valid:)
   > n16 = Branch(Condition=@n11)
   n17 = Branch(execute=[@n16.then, @n19.then], Condition=@n12)
   n18 = Set AnimSectionStatus(execute=@n17.else, AnimSectionStatus=Append!(A=Get AnimSectionStatus(), B=" [End Frame] ") -> (ReturnValue))
   n19 = Set AnimSectionStatus(execute=@n16.else, AnimSectionStatus=Append(A=Get AnimSectionStatus(), B=" [Start Frame] "))
   CreateAnimSeqRow(execute=@n9.then_1, AnimSequence=Get AnimSeqRef(), StatusText=@n10.StatusText, IsValid=@n10)
   int / int(A=@n10.RequiredStartFrame, B=60)
   Get CheckedOutAssets()
  ]}
 WaitForOpen {nodes [
   n1 = Values!(execute, TargetMap:Map of Strings to Object References=Get CheckedOutAssets()) -> (then, Values:Array of Object References)
   > n2 = For Each Loop(Array=@n1)
   > n3 = Cast To LevelSequence(Object=@n2)
   n4 = ValidateWIPSeq(self=Get Validator(), Project=Get Current Project(), LevelSequence=@n3, Range Start=0, Range End=0)
   > Branch(Condition=@n4)
   > CreateLvlSeqRow(LevelSequence=@n3, StatusText=@n4.StatusText, IsValid=@n4.IsValid)
   WaitForOpen() {is=title K2Node_FunctionEntry}
   > n5 = GetCurrentLevelSequence!(execute) -> (then, ReturnValue:Level Sequence Object Reference)
   > Is Valid(InputObject=@n5)
   > Branch(Condition="Equal (Object)"(A=@n5, B=Get ProjectRootSequence() -> (ProjectRootSequence:Level Sequence Object Reference)))
   > Clear and Invalidate Timer by Handle(Handle=Get Timer() -> (Timer:Timer Handle Structure))
   > n6 = Sequence()
   GetSubSequences(execute=@n6.then_1, Sequence=Get ProjectRootSequence())
  ]}
]
variables [
 AssetsToValidate {type="array<SourceControlState>", default []}
 CheckedOutAssets {type="map<Object>", default {}}
 CheckedSequences {type="array<string>", default []}
 ConstructTimer {type=TimerHandle}
 Current Project {type=BP_AnimationBrowser_Project_C, default=null}
 CurrentProjects {type="array<BP_AnimationBrowser_Project_C>", default []}
 CurrentTableRow {type=Struct_EventsValidator, default {character {}}}
 ProjectAC {type=AnimatableCreature_C, default=null}
 ProjectDataTable {type=DataTable, default=null}
 ProjectNames {type="array<string>", default []}
 ProjectRootSequence {type=LevelSequence, default=null}
 Timer {type=TimerHandle}
 Validator {type=EventValidatorFunctions_C, default=null}
]
widgets {
 info {rootWidgetClass=&CanvasPanel, widgetCount=11}
 widgets [
  CanvasPanel_0 {widgetClassPath=&CanvasPanel, widget=CanvasPanel_0}
  VerticalBox_0 {
   widgetClassPath=&VerticalBox
   parent=CanvasPanel_0
   widget=VerticalBox_0
   slot {class=CanvasPanelSlot, properties {layoutData {alignment=<0 0>, anchors {maximum=<1 1>, minimum=<0 0>}, offsets=<ltrb 0 0 11.4955 9.009>}}}
  }
  HeaderBorder {
   widgetClassPath=&Border
   parent=VerticalBox_0
   widget=HeaderBorder
   properties {brushColor=<rgba 0.0275 0.0275 0.0275 1>, padding=<ltrb 2 2 2 2>}
   slot {class=VerticalBoxSlot, properties {padding=<ltrb 0 0 0 1>}}
  }
  VerticalBox_1 {
   widgetClassPath=&VerticalBox
   parent=HeaderBorder
   widget=VerticalBox_1
   slot {class=BorderSlot, properties {padding=<ltrb 2 2 2 2>}}
  }
  HorizontalBox_1 {widgetClassPath=&HorizontalBox, parent=VerticalBox_1, widget=HorizontalBox_1, slot {class=VerticalBoxSlot}}
  AssetImage {
   widgetClassPath=&Image
   bIsVariable=true
   parent=HorizontalBox_1
   widget=AssetImage
   properties {brush {
     bIsDynamicallyLoaded=false
     drawAs=Image
     imageType=FullColor
     mirroring=NoMirror
     resourceName=None
     resourceObject=/AnimInEdTools/AnimationBrowser/Icons/Icon_Asset
     tiling=NoTile
     imageSize=<18 28>
     margin=<ltrb 0 0 0 0>
     outlineSettings {
      bUseBrushTransparency=false
      roundingType=HalfHeightRadius
      width=0
      color {colorUseRule=UseColor_Specified, specifiedColor=<rgba 0 0 0 0>}
      cornerRadii=<0 0 0 0>
     }
     tintColor {colorUseRule=UseColor_Specified, specifiedColor=<rgba 1 1 1 1>}
     uVRegion {bIsValid=false, max=<0 0>, min=<0 0>}
    }}
   slot {class=HorizontalBoxSlot, properties {horizontalAlignment=HAlign_Center, verticalAlignment=VAlign_Center}}
  }
  SizeBox_3 {
   widgetClassPath=&SizeBox
   parent=HorizontalBox_1
   widget=SizeBox_3
   properties {
    bOverride_MaxDesiredHeight=true
    bOverride_MaxDesiredWidth=true
    bOverride_MinDesiredWidth=true
    maxDesiredHeight=25
    maxDesiredWidth=200
    minDesiredWidth=100
   }
   slot {class=HorizontalBoxSlot, properties {size {sizeRule=Fill, value=1}}}
  }
  Project_ComboBox {
   widgetClassPath=/Script/FortniteEditor.SearchableStringComboBox
   bIsVariable=true
   parent=SizeBox_3
   widget=Project_ComboBox
   properties {
    selectedOption=Project
    defaultOptions [Project]
    font {
     bForceMonospaced=false
     bMaterialIsStencil=false
     fontMaterial=null
     fontObject=&Roboto
     letterSpacing=0
     monospacedWidth=1
     size=12
     skewAmount=0
     typefaceFontName=Bold
     outlineSettings {
      bApplyOutlineToDropShadows=false
      bMiteredCorners=false
      bSeparateFillAlpha=false
      outlineMaterial=null
      outlineSize=0
      outlineColor=<rgba 0 0 0 1>
     }
    }
    foregroundColor {colorUseRule=UseColor_Specified, specifiedColor=<rgba 0.5271 0.5271 0.5271 1>}
    itemStyle {
     bUseParentRowBrush=false
     activeBrush {
      bIsDynamicallyLoaded=false
      drawAs=Image
      imageType=NoImage
      mirroring=NoMirror
      resourceName=None
      resourceObject=null
      tiling=NoTile
      imageSize=<0 0>
      margin=<ltrb 0 0 0 0>
      outlineSettings {
       bUseBrushTransparency=false
       roundingType=HalfHeightRadius
       width=0
       color {colorUseRule=UseColor_Specified, specifiedColor=<rgba 0 0 0 0>}
       cornerRadii=<0 0 0 0>
      }
      tintColor {colorUseRule=UseColor_Specified, specifiedColor=<rgba 0 0.162 0.7454 1>}
      uVRegion {bIsValid=false, max=<0 0>, min=<0 0>}
     }
     activeHighlightedBrush {
      bIsDynamicallyLoaded=false
      drawAs=Image
      imageType=NoImage
      mirroring=NoMirror
      resourceName=None
      resourceObject=null
      tiling=NoTile
      imageSize=<0 0>
      margin=<ltrb 0 0 0 0>
      outlineSettings {
       bUseBrushTransparency=false
       roundingType=HalfHeightRadius
       width=0
       color {colorUseRule=UseColor_Specified, specifiedColor=<rgba 0 0 0 0>}
       cornerRadii=<0 0 0 0>
      }
      tintColor {colorUseRule=UseColor_Specified, specifiedColor=<rgba 0.0252 0.0319 0.0423 1>}
      uVRegion {bIsValid=false, max=<0 0>, min=<0 0>}
     }
     activeHoveredBrush {
      bIsDynamicallyLoaded=false
      drawAs=Image
      imageType=NoImage
      mirroring=NoMirror
      resourceName=None
      resourceObject=null
      tiling=NoTile
      imageSize=<0 0>
      margin=<ltrb 0 0 0 0>
      outlineSettings {
       bUseBrushTransparency=false
       roundingType=HalfHeightRadius
       width=0
       color {colorUseRule=UseColor_Specified, specifiedColor=<rgba 0 0 0 0>}
       cornerRadii=<0 0 0 0>
      }
      tintColor {colorUseRule=UseColor_Specified, specifiedColor=<rgba 0 0.162 0.7454 1>}
      uVRegion {bIsValid=false, max=<0 0>, min=<0 0>}
     }
     dropIndicator_Above {
      bIsDynamicallyLoaded=false
      drawAs=Box
      imageType=FullColor
      mirroring=NoMirror
      resourceName=../../../Engine/Content/Slate/Common/DropZoneIndicator_Above.png
      resourceObject=null
      tiling=NoTile
      imageSize=<0 0>
      margin=<ltrb 0.625 0.625 0 0>
      outlineSettings {
       bUseBrushTransparency=false
       roundingType=HalfHeightRadius
       width=0
       color {colorUseRule=UseColor_Specified, specifiedColor=<rgba 0 0 0 0>}
       cornerRadii=<0 0 0 0>
      }
      tintColor {colorUseRule=UseColor_Specified, specifiedColor=<rgba 0.0091 0.3515 0.9647 1>}
      uVRegion {bIsValid=false, max=<0 0>, min=<0 0>}
     }
     dropIndicator_Below {
      bIsDynamicallyLoaded=false
      drawAs=Box
      imageType=FullColor
      mirroring=NoMirror
      resourceName=../../../Engine/Content/Slate/Common/DropZoneIndicator_Below.png
      resourceObject=null
      tiling=NoTile
      imageSize=<0 0>
      margin=<ltrb 0.625 0 0 0.625>
      outlineSettings {
       bUseBrushTransparency=false
       roundingType=HalfHeightRadius
       width=0
       color {colorUseRule=UseColor_Specified, specifiedColor=<rgba 0 0 0 0>}
       cornerRadii=<0 0 0 0>
      }
      tintColor {colorUseRule=UseColor_Specified, specifiedColor=<rgba 0.0091 0.3515 0.9647 1>}
      uVRegion {bIsValid=false, max=<0 0>, min=<0 0>}
     }
     dropIndicator_Onto {
      bIsDynamicallyLoaded=false
      drawAs=Box
      imageType=FullColor
      mirroring=NoMirror
      resourceName=../../../Engine/Content/Slate/Common/DropZoneIndicator_Onto.png
      resourceObject=null
      tiling=NoTile
      imageSize=<0 0>
      margin=<ltrb 0.25 0.25 0.25 0.25>
      outlineSettings {
       bUseBrushTransparency=false
       roundingType=HalfHeightRadius
       width=0
       color {colorUseRule=UseColor_Specified, specifiedColor=<rgba 0 0 0 0>}
       cornerRadii=<0 0 0 0>
      }
      tintColor {colorUseRule=UseColor_Specified, specifiedColor=<rgba 0.0091 0.3515 0.9647 1>}
      uVRegion {bIsValid=false, max=<0 0>, min=<0 0>}
     }
     evenRowBackgroundBrush {
      bIsDynamicallyLoaded=false
      drawAs=Image
      imageType=NoImage
      mirroring=NoMirror
      resourceName=None
      resourceObject=null
      tiling=NoTile
      imageSize=<0 0>
      margin=<ltrb 0 0 0 0>
      outlineSettings {
       bUseBrushTransparency=false
       roundingType=HalfHeightRadius
       width=0
       color {colorUseRule=UseColor_Specified, specifiedColor=<rgba 0 0 0 0>}
       cornerRadii=<0 0 0 0>
      }
      tintColor {colorUseRule=UseColor_Specified, specifiedColor=<rgba 0.0103 0.0103 0.0103 1>}
      uVRegion {bIsValid=false, max=<0 0>, min=<0 0>}
     }
     evenRowBackgroundHoveredBrush {
      bIsDynamicallyLoaded=false
      drawAs=Image
      imageType=NoImage
      mirroring=NoMirror
      resourceName=None
      resourceObject=null
      tiling=NoTile
      imageSize=<0 0>
      margin=<ltrb 0 0 0 0>
      outlineSettings {
       bUseBrushTransparency=false
       roundingType=HalfHeightRadius
       width=0
       color {colorUseRule=UseColor_Specified, specifiedColor=<rgba 0 0 0 0>}
       cornerRadii=<0 0 0 0>
      }
      tintColor {colorUseRule=UseColor_Specified, specifiedColor=<rgba 0.0176 0.0176 0.0176 1>}
      uVRegion {bIsValid=false, max=<0 0>, min=<0 0>}
     }
     inactiveBrush {
      bIsDynamicallyLoaded=false
      drawAs=Image
      imageType=NoImage
      mirroring=NoMirror
      resourceName=None
      resourceObject=null
      tiling=NoTile
      imageSize=<0 0>
      margin=<ltrb 0 0 0 0>
      outlineSettings {
       bUseBrushTransparency=false
       roundingType=HalfHeightRadius
       width=0
       color {colorUseRule=UseColor_Specified, specifiedColor=<rgba 0 0 0 0>}
       cornerRadii=<0 0 0 0>
      }
      tintColor {colorUseRule=UseColor_Specified, specifiedColor=<rgba 0.0513 0.0953 0.159 1>}
      uVRegion {bIsValid=false, max=<0 0>, min=<0 0>}
     }
     inactiveHighlightedBrush {
      bIsDynamicallyLoaded=false
      drawAs=Image
      imageType=NoImage
      mirroring=NoMirror
      resourceName=None
      resourceObject=null
      tiling=NoTile
      imageSize=<0 0>
      margin=<ltrb 0 0 0 0>
      outlineSettings {
       bUseBrushTransparency=false
       roundingType=HalfHeightRadius
       width=0
       color {colorUseRule=UseColor_Specified, specifiedColor=<rgba 0 0 0 0>}
       cornerRadii=<0 0 0 0>
      }
      tintColor {colorUseRule=UseColor_Specified, specifiedColor=<rgba 0.0252 0.0319 0.0423 1>}
      uVRegion {bIsValid=false, max=<0 0>, min=<0 0>}
     }
     inactiveHoveredBrush {
      bIsDynamicallyLoaded=false
      drawAs=Image
      imageType=NoImage
      mirroring=NoMirror
      resourceName=None
      resourceObject=null
      tiling=NoTile
      imageSize=<0 0>
      margin=<ltrb 0 0 0 0>
      outlineSettings {
       bUseBrushTransparency=false
       roundingType=HalfHeightRadius
       width=0
       color {colorUseRule=UseColor_Specified, specifiedColor=<rgba 0 0 0 0>}
       cornerRadii=<0 0 0 0>
      }
      tintColor {colorUseRule=UseColor_Specified, specifiedColor=<rgba 0.0513 0.0953 0.159 1>}
      uVRegion {bIsValid=false, max=<0 0>, min=<0 0>}
     }
     oddRowBackgroundBrush {
      bIsDynamicallyLoaded=false
      drawAs=Image
      imageType=NoImage
      mirroring=NoMirror
      resourceName=None
      resourceObject=null
      tiling=NoTile
      imageSize=<0 0>
      margin=<ltrb 0 0 0 0>
      outlineSettings {
       bUseBrushTransparency=false
       roundingType=HalfHeightRadius
       width=0
       color {colorUseRule=UseColor_Specified, specifiedColor=<rgba 0 0 0 0>}
       cornerRadii=<0 0 0 0>
      }
      tintColor {colorUseRule=UseColor_Specified, specifiedColor=<rgba 0.0103 0.0103 0.0103 1>}
      uVRegion {bIsValid=false, max=<0 0>, min=<0 0>}
     }
     oddRowBackgroundHoveredBrush {
      bIsDynamicallyLoaded=false
      drawAs=Image
      imageType=NoImage
      mirroring=NoMirror
      resourceName=None
      resourceObject=null
      tiling=NoTile
      imageSize=<0 0>
      margin=<ltrb 0 0 0 0>
      outlineSettings {
       bUseBrushTransparency=false
       roundingType=HalfHeightRadius
       width=0
       color {colorUseRule=UseColor_Specified, specifiedColor=<rgba 0 0 0 0>}
       cornerRadii=<0 0 0 0>
      }
      tintColor {colorUseRule=UseColor_Specified, specifiedColor=<rgba 0.0176 0.0176 0.0176 1>}
      uVRegion {bIsValid=false, max=<0 0>, min=<0 0>}
     }
     parentRowBackgroundBrush {
      bIsDynamicallyLoaded=false
      drawAs=Image
      imageType=NoImage
      mirroring=NoMirror
      resourceName=None
      resourceObject=null
      tiling=NoTile
      imageSize=<32 32>
      margin=<ltrb 0 0 0 0>
      outlineSettings {
       bUseBrushTransparency=false
       roundingType=HalfHeightRadius
       width=0
       color {colorUseRule=UseColor_Specified, specifiedColor=<rgba 0 0 0 0>}
       cornerRadii=<0 0 0 0>
      }
      tintColor {colorUseRule=UseColor_Specified, specifiedColor=<rgba 1 1 1 1>}
      uVRegion {bIsValid=false, max=<0 0>, min=<0 0>}
     }
     parentRowBackgroundHoveredBrush {
      bIsDynamicallyLoaded=false
      drawAs=Image
      imageType=NoImage
      mirroring=NoMirror
      resourceName=None
      resourceObject=null
      tiling=NoTile
      imageSize=<32 32>
      margin=<ltrb 0 0 0 0>
      outlineSettings {
       bUseBrushTransparency=false
       roundingType=HalfHeightRadius
       width=0
       color {colorUseRule=UseColor_Specified, specifiedColor=<rgba 0 0 0 0>}
       cornerRadii=<0 0 0 0>
      }
      tintColor {colorUseRule=UseColor_Specified, specifiedColor=<rgba 1 1 1 1>}
      uVRegion {bIsValid=false, max=<0 0>, min=<0 0>}
     }
     selectedTextColor {colorUseRule=UseColor_Specified, specifiedColor=<rgba 1 1 1 1>}
     selectorFocusedBrush {
      bIsDynamicallyLoaded=false
      drawAs=Border
      imageType=FullColor
      mirroring=NoMirror
      resourceName=../../../Engine/Content/Slate/Common/Selector.png
      resourceObject=null
      tiling=Both
      imageSize=<0 0>
      margin=<ltrb 0.25 0.25 0.25 0.25>
      outlineSettings {
       bUseBrushTransparency=false
       roundingType=HalfHeightRadius
       width=0
       color {colorUseRule=UseColor_Specified, specifiedColor=<rgba 0 0 0 0>}
       cornerRadii=<0 0 0 0>
      }
      tintColor {colorUseRule=UseColor_Specified, specifiedColor=<rgba 0.701 0.225 0.003 1>}
      uVRegion {bIsValid=false, max=<0 0>, min=<0 0>}
     }
     textColor {colorUseRule=UseColor_Specified, specifiedColor=<rgba 0.5271 0.5271 0.5271 1>}
    }
    widgetStyle {
     comboButtonStyle {
      downArrowAlign=VAlign_Center
      buttonStyle {
       clickedSlateSound {resourceObject=null}
       disabled {
        bIsDynamicallyLoaded=false
        drawAs=RoundedBox
        imageType=NoImage
        mirroring=NoMirror
        resourceName=None
        resourceObject=null
        tiling=NoTile
        imageSize=<0 0>
        margin=<ltrb 0 0 0 0>
        outlineSettings {
         bUseBrushTransparency=false
         roundingType=FixedRadius
         width=1
         color {colorUseRule=UseColor_ColorTable, specifiedColor=<rgba 0 0 0 0>}
         cornerRadii=<4 4 4 4>
        }
        tintColor {colorUseRule=UseColor_Specified, specifiedColor=<rgba 0.0395 0.0395 0.0395 1>}
        uVRegion {bIsValid=false, max=<0 0>, min=<0 0>}
       }
       disabledForeground {colorUseRule=UseColor_ColorTable, specifiedColor=<rgba 0 0 0 0>}
       hovered {
        bIsDynamicallyLoaded=false
        drawAs=RoundedBox
        imageType=NoImage
        mirroring=NoMirror
        resourceName=None
        resourceObject=null
        tiling=NoTile
        imageSize=<0 0>
        margin=<ltrb 0 0 0 0>
        outlineSettings {
         bUseBrushTransparency=false
         roundingType=FixedRadius
         width=1
         color {colorUseRule=UseColor_ColorTable, specifiedColor=<rgba 0 0 0 0>}
         cornerRadii=<4 4 4 4>
        }
        tintColor {colorUseRule=UseColor_Specified, specifiedColor=<rgba 0.0048 0.0048 0.0048 1>}
        uVRegion {bIsValid=false, max=<0 0>, min=<0 0>}
       }
       hoveredForeground {colorUseRule=UseColor_ColorTable, specifiedColor=<rgba 0 0 0 0>}
       hoveredSlateSound {resourceObject=null}
       normal {
        bIsDynamicallyLoaded=false
        drawAs=RoundedBox
        imageType=NoImage
        mirroring=NoMirror
        resourceName=None
        resourceObject=null
        tiling=NoTile
        imageSize=<0 0>
        margin=<ltrb 0 0 0 0>
        outlineSettings {
         bUseBrushTransparency=false
         roundingType=FixedRadius
         width=1
         color {colorUseRule=UseColor_ColorTable, specifiedColor=<rgba 0 0 0 0>}
         cornerRadii=<4 4 4 4>
        }
        tintColor {colorUseRule=UseColor_Specified, specifiedColor=<rgba 0.0048 0.0048 0.0048 1>}
        uVRegion {bIsValid=false, max=<0 0>, min=<0 0>}
       }
       normalForeground {colorUseRule=UseColor_ColorTable, specifiedColor=<rgba 0 0 0 0>}
       normalPadding=<ltrb 8 1 8 1>
       pressed {
        bIsDynamicallyLoaded=false
        drawAs=RoundedBox
        imageType=NoImage
        mirroring=NoMirror
        resourceName=None
        resourceObject=null
        tiling=NoTile
        imageSize=<0 0>
        margin=<ltrb 0 0 0 0>
        outlineSettings {
         bUseBrushTransparency=false
         roundingType=FixedRadius
         width=1
         color {colorUseRule=UseColor_ColorTable, specifiedColor=<rgba 0 0 0 0>}
         cornerRadii=<4 4 4 4>
        }
        tintColor {colorUseRule=UseColor_Specified, specifiedColor=<rgba 0.0048 0.0048 0.0048 1>}
        uVRegion {bIsValid=false, max=<0 0>, min=<0 0>}
       }
       pressedForeground {colorUseRule=UseColor_ColorTable, specifiedColor=<rgba 0 0 0 0>}
       pressedPadding=<ltrb 8 2 8 0>
       pressedSlateSound {resourceObject=null}
      }
      contentPadding=<ltrb 0 0 0 0>
      downArrowImage {
       bIsDynamicallyLoaded=false
       drawAs=Image
       imageType=Vector
       mirroring=NoMirror
       resourceName=../../../Engine/Content/Slate/Starship/CoreWidgets/ComboBox/wide-chevron-down.svg
       resourceObject=null
       tiling=NoTile
       imageSize=<16 16>
       margin=<ltrb 0 0 0 0>
       outlineSettings {
        bUseBrushTransparency=false
        roundingType=HalfHeightRadius
        width=0
        color {colorUseRule=UseColor_Specified, specifiedColor=<rgba 0 0 0 0>}
        cornerRadii=<0 0 0 0>
       }
       tintColor {colorUseRule=UseColor_Specified, specifiedColor=<rgba 1 1 1 1>}
       uVRegion {bIsValid=false, max=<0 0>, min=<0 0>}
      }
      downArrowPadding=<ltrb 2 2 2 2>
      menuBorderBrush {
       bIsDynamicallyLoaded=false
       drawAs=RoundedBox
       imageType=NoImage
       mirroring=NoMirror
       resourceName=None
       resourceObject=null
       tiling=NoTile
       imageSize=<0 0>
       margin=<ltrb 0 0 0 0>
       outlineSettings {
        bUseBrushTransparency=false
        roundingType=FixedRadius
        width=1
        color {colorUseRule=UseColor_ColorTable, specifiedColor=<rgba 0 0 0 0>}
        cornerRadii=<0 0 0 0>
       }
       tintColor {colorUseRule=UseColor_Specified, specifiedColor=<rgba 0.0395 0.0395 0.0395 1>}
       uVRegion {bIsValid=false, max=<0 0>, min=<0 0>}
      }
      menuBorderPadding=<ltrb 0 0 0 0>
      shadowColorAndOpacity=<rgba 0 0 0 1>
      shadowOffset=<0 0>
     }
     contentPadding=<ltrb 0 0 0 0>
     menuRowPadding=<ltrb 8 3 8 3>
     pressedSlateSound {resourceObject=null}
     selectionChangeSlateSound {resourceObject=null}
    }
   }
   slot {class=SizeBoxSlot}
  }
  ScrollBox {
   widgetClassPath=&EditorUtilityScrollBox
   bIsVariable=true
   parent=VerticalBox_0
   widget=ScrollBox
   slot {class=VerticalBoxSlot}
  }
  ValidateBtn {
   widgetClassPath=&EditorUtilityButton
   parent=CanvasPanel_0
   widget=ValidateBtn
   slot {
    class=CanvasPanelSlot
    properties {layoutData {alignment=<0 0>, anchors {maximum=<1 1>, minimum=<0 1>}, offsets=<ltrb 0 -48.9377 1.8071 48.4787>}}
   }
  }
  ValidateTxt {widgetClassPath=&TextBlock, parent=ValidateBtn, widget=ValidateTxt, properties {text=Validate}, slot {class=ButtonSlot}}
 ]
}
